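# TNTina skydiving contrail Niagara system: speed-line ribbons, glows and lights, with heart, firework, smoke and "Ka"/"Boom" mesh explosion effects.
# Attached sprites, lights and short-lived parent particles follow bones sampled from the user skeletal mesh.
# Parents and a flickering big heart generate death events, and small hearts send location events.
# Explosion emitters have their own spawn modules disabled; curves drive their scale, color and dynamic material parameters.
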NiagaraSystem /BRCosmetics/Athena/Effects/Particles/Skydiving/NS_Contrail_TNTina/NS_Contrail_TNTina_V2
dependencies {
 usedDataInterfaces [
  &NiagaraDataInterfaceCurve
  &NiagaraDataInterfaceColorCurve
  &NiagaraDataInterfaceVectorCurve
  &NiagaraDataInterfaceVector2DCurve
 ]
 usedDynamicInputs []
 usedModules [&SystemState]
}
emitters {
 Contrail_SpeedLine {
  emitterUpdateScript [
   &EmitterScalability(
    Module.bApplySpawnCountScale=false
    Module.bApplySpawnCountScaleByDistanceFraction=false
    Module.bUseMaxDistance=false
    Module.bUseMinDistance=false
   ) {enabled=false}
   &EmitterLifeCycle(CompleteOnInactive=true) {enabled=false}
   &EmitterState(Enable Distance Culling=true, MaxDistance=2000, Module.bUseMaxDistance=true, Module.bUseMinDistance=false, Scalability Mode=::NewEnumerator1)
   /Game/Effects/Niagara/Modules/Spawn/FN_SpawnContrails(All limbs=false, Global Spawnrate=10) {enabled=false}
   &FN_SpawnRibbon_v2(Number of Ribbons=4, Spawn Mode=::NewEnumerator0)
  ]
  particleSpawnScript [
   &FN_InitializeRibbon_v3(Enable Skeletal Position=true, Lifetime=1, Ribbon Width=20)
   &AddVelocity(Velocity=&FN_FrontEndVector3Modifier(Gameplay=<0 0 300>, Front End=<0 0 300>))
   /SetVariables(Particles.FadeOffAlpha:NiagaraFloat=&ClampFloat(Float=&Multiply_Float(A=$Emitter.Age, B=1))) {isSetParameters=true}
  ]
  particleUpdateScript [
   /Game/Effects/Niagara/Modules/Skydiving/FN_DivingAttributes/FN_DivingAttributes(Accumulate Dive=0, Accumulate Rest=0) {enabled=false}
   &SolveForcesAndVelocity(Clamp Velocity=false, Limit Acceleration=false)
   &FN_UpdateRibbon(
    Enable Skeletal Position=true
    Initial Color:LinearColor=$Particles.Initial.Color
    Initial RibbonWidth:NiagaraFloat=$Particles.Initial.RibbonWidth
   )
   &Material(Base Texture Time:NiagaraFloat=&Multiply_Float(A=$Emitter.Age, B=-1))
   &DynamicMaterialParameters(
    Index 0 Param 1=1
    Index 0 Param 2=1
    Index 0 Param 3=0.1
    Index 0 Param 4=0.2
    Index 1 Param 1=&Add_Float(
     A=&Multiply_Float(A=1, B=&MakeFloatFromInt(Int=HlslExpression() {hlslExpression=Particles.RibbonID}))
     B=&Multiply_Float(A=-1, B=$Engine.Time)
    )
    Index 1 Param 2=0.15
    Index 1 Param 3=1.3
    Index 1 Param 4=1
    Index 2 Param 1=0.2
    Index 2 Param 2=0.2
    Index 2 Param 3=0.2
    Index 3 Param 1=1
    Index 3 Param 2=1
    Index 3 Param 3=1
    Index 3 Param 4=1
    Module.Param0WriteEnabled=true
    Module.Param0WriteEnabled1=true
    Module.Param0WriteEnabled2=true
    Module.Param0WriteEnabled3=true
    Module.Param1WriteEnabled=true
    Module.Param1WriteEnabled1=true
    Module.Param1WriteEnabled2=true
    Module.Param1WriteEnabled3=true
    Module.Param2WriteEnabled=true
    Module.Param2WriteEnabled1=true
    Module.Param2WriteEnabled2=true
    Module.Param2WriteEnabled3=true
    Module.Param3WriteEnabled=true
    Module.Param3WriteEnabled1=true
    Module.Param3WriteEnabled2=true
    Module.Param3WriteEnabled3=true
    Write Parameter Index 1=true
    Write Parameter Index 2=true
    Write Parameter Index 3=true
   )
   &ScaleRibbonWidth(Ribbon Width Scale=&FloatFromCurve(FloatCurve=DataInterface() {propertyValues {Curve {keys [
         <at 0.0039 1.3394 RCIM_Cubic>
         {arriveTangent=-0.4407, interpMode=RCIM_Cubic, leaveTangent=-0.4407, time=0.1718, value=1.5109}
         <at 1.0045 0.8984>
        ]}}}))
   &Color(
    Color=&ColorFromCurve(ColorCurve=DataInterface() {propertyValues {
       AlphaCurve {keys [<at -0.0023 1 RCIM_Cubic>,<at 0.9977 1 RCIM_Cubic>]}
       BlueCurve {keys [<at -0.0023 0.1137 RCIM_Cubic>,<at 0.6301 0.201>,<at 1.0058 0.0534 RCIM_Cubic>]}
       GreenCurve {keys [<at -0.0023 0.0671 RCIM_Cubic>,<at 0.6301 0.0375>,<at 1.0058 0 RCIM_Cubic>]}
       RedCurve {keys [<at -0.0023 1 RCIM_Cubic>,<at 0.6301 2>,<at 1.0058 0.2135 RCIM_Cubic>]}
      }})
    Scale Alpha=&FloatFromCurve(
     FloatCurve=DataInterface() {propertyValues {Curve {keys [{arriveTangent=-2.0086, interpMode=RCIM_Cubic, leaveTangent=-2.0086, tangentMode=RCTM_User, time=-0.0046, value=1.0113}, <at 1.0011 0.002>]}}}
     Scale Curve=&FloatFromCurve(FloatCurve=DataInterface() {propertyValues {Curve {keys [<at 0.3 0>,<at 0.9974 0.1>]}}}, CurveIndex=$Emitter.Age)
    )
    Scale Color=<4 4 1>
   )
  ]
  rendererClasses [&NiagaraSpriteRendererProperties, &NiagaraRibbonRendererProperties]
  renderers [
   {rendererClass=&NiagaraSpriteRendererProperties}
   {rendererClass=&NiagaraRibbonRendererProperties, rendererIndex=1}
  ]
 }
 FN_Contrails_AttachedParticle {
  emitterSpawnScript [&FN_PreserveSkydiveUserVariables() {module=PreserveUserVariables}]
  emitterUpdateScript [
   &EmitterState()
   /Game/Effects/Niagara/Modules/Spawn/SpawnRateFromMultipleSpawnGroups_2(Global Spawnrate=2)
   &SpawnBurst_Instantaneous(Spawn Count=4, Use Spawn Probability=false) {enabled=false}
  ]
  particleSpawnScript [
   /Game/Effects/Niagara/Modules/Groups/FN_SetSpawnGroup_v1/FN_SetSpawnGroup(Emitter Bone Count=2)
   /Niagara/Modules/Spawn/Initialization/InitializeParticle(
    Lifetime=0.6
    Module.Write Color=true
    Module.Write Lifetime=true
    Module.Write Mass=true
    Module.Write Material Random=true
    Module.Write Position=false
    Module.Write Scale=false
    Module.Write SpriteRotation=false
    Module.Write SpriteSize=true
    Position=&SimulationPosition()
    Sprite Size=&Vector2DFromFloat(Value=&UniformRangedFloat(Minimum=40, Maximum=80))
   )
   &CameraOffset(Camera Offset Amount=&UniformRangedFloat(Minimum=10, Maximum=30))
  ]
  particleUpdateScript [
   /SetVariables(Particles.Color=&MakeLinearColorFromVectorAndFloat(
     "Vector (RGB)"=&MakeVectorFromLinearColorRGB(Color=&ColorFromCurve(ColorCurve=DataInterface() {propertyValues {
         AlphaCurve {keys [<at -0.0023 1 RCIM_Cubic>,<at 0.9977 1 RCIM_Cubic>]}
         BlueCurve {keys [<at -0.0036 0.0782 RCIM_Cubic>,<at 0.2978 0.3631>,<at 0.6731 0.1667>,<at 1.034 0.0782>,<at 1.0651 0 RCIM_Cubic>]}
         GreenCurve {keys [<at -0.0036 0.1144 RCIM_Cubic>,<at 0.2978 0.2143>,<at 0.6731 0>,<at 1.034 0.1144>,<at 1.0651 0.4573 RCIM_Cubic>]}
         RedCurve {keys [<at -0.0036 1 RCIM_Cubic>,<at 0.2978 2>,<at 0.6731 1>,<at 1.034 1>,<at 1.0651 1 RCIM_Cubic>]}
        }}))
     "Float (Alpha)"=&FloatFromCurve(FloatCurve=DataInterface() {propertyValues {Curve {keys [<at 0 0>,<at 0.07 0.3>,<at 0.8787 0.3>,<at 1.0011 0.002>]}}})
    )) {isSetParameters=true}
   &ParticleState()
   &SampleSkeletalMeshSkeleton(Mesh:NiagaraDataInterfaceSkeletalMesh=$User.SkeletalMesh, Specific Bone Index:NiagaraInt32=$Particles.SpawnGroup)
   /Niagara/Modules/Spawn/Location/SkeletalMeshSkeletonLocation(Module.Inherit Velocity=false, Module.Use Sampled Bone as Ribbon ID=true)
   &SolveForcesAndVelocity(Clamp Velocity=false, Limit Acceleration=false) {enabled=false}
   &ScaleMeshSize(Scale Factor=&Multiply_Vector(
     A=<1 1 1>
     B=&VectorFromFloat(Value=&FloatFromCurve(FloatCurve=DataInterface() {propertyValues {Curve {keys [<at 0 0>,<at 1 1.5>]}}}))
    ))
  ]
  rendererClasses [&NiagaraSpriteRendererProperties]
  renderers [{rendererClass=&NiagaraSpriteRendererProperties}]
 }
 FN_Contrails_LandR_Parents {
  emitterSpawnScript [&FN_PreserveSkydiveUserVariables() {module=PreserveUserVariables}]
  emitterUpdateScript [
   &EmitterState(
    Delay First Loop Only=true
    Inactive Response=::NewEnumerator1
    Life Cycle Mode=::NewEnumerator1
    Loop Delay=1
    Loop Duration=2
    Scalability Mode=::NewEnumerator1
    UseLoopDelay=true
   )
   &SpawnRate(Use Spawn Probability=false) {enabled=false}
   &SpawnBurst_Instantaneous(Spawn Count=1, Use Spawn Probability=false)
  ]
  particleSpawnScript [
   &FN_SetSpawnGroup(Emitter Bone Count=2)
   &InitializeParticle(
    Color=<rgba 1 0.8417 0.7639 1>
    Lifetime=0.1
    Mesh Scale=&Multiply_Vector(A=<1 1 1>, B=<1 1 1>)
    Module.Write Color=false
    Module.Write Lifetime=true
    Module.Write Mass=true
    Module.Write Material Random=true
    Module.Write Position=false
    Module.Write Scale=false
    Module.Write SpriteRotation=false
    Module.Write SpriteSize=true
    Position=&SimulationPosition()
    Sprite Size=<50 50>
   )
   &SampleSkeletalMeshSkeleton(
    Mesh:NiagaraDataInterfaceSkeletalMesh=$User.SkeletalMesh
    Specific Bone Index:NiagaraInt32=&MakeIntFromFloat(Float=/Game/Effects/Niagara/DynamicInputs/FN_PickFloatFromSpawnGroupCutoff(Over Cutoff Float:NiagaraFloat=1, Under Cutoff Float:NiagaraFloat=0, SpawnGroup:NiagaraInt32=$Particles.SpawnGroup, Cutoff:NiagaraInt32=1))
   )
   &SkeletalMeshSkeletonLocation(Module.Inherit Velocity=false, Module.Use Sampled Bone as Ribbon ID=true)
  ]
  particleUpdateScript [
   &ParticleState()
   &SolveForcesAndVelocity(Clamp Velocity=false, Limit Acceleration=false)
   &GenerateDeathEvent()
  ]
  rendererClasses [&NiagaraSpriteRendererProperties]
  renderers [{rendererClass=&NiagaraSpriteRendererProperties}]
 }
 FN_Contrails_LandR_Parents_delayed {
  emitterSpawnScript [&FN_PreserveSkydiveUserVariables() {module=PreserveUserVariables}]
  emitterUpdateScript [
   &EmitterState(
    Delay First Loop Only=true
    Inactive Response=::NewEnumerator1
    Life Cycle Mode=::NewEnumerator1
    Loop Delay=1
    Loop Duration=2
    Scalability Mode=::NewEnumerator1
    UseLoopDelay=true
   )
   &SpawnRate(Use Spawn Probability=false) {enabled=false}
   &SpawnBurst_Instantaneous(Spawn Count=1, Use Spawn Probability=false)
  ]
  particleSpawnScript [
   &FN_SetSpawnGroup(Emitter Bone Count=2)
   &InitializeParticle(
    Color=<rgba 1 0.8417 0.7639 1>
    Lifetime=0.4
    Mesh Scale=&Multiply_Vector(A=<1 1 1>, B=<1 1 1>)
    Module.Write Color=false
    Module.Write Lifetime=true
    Module.Write Mass=true
    Module.Write Material Random=true
    Module.Write Position=false
    Module.Write Scale=false
    Module.Write SpriteRotation=false
    Module.Write SpriteSize=true
    Position=&SimulationPosition()
    Sprite Size=<50 50>
   )
   &SampleSkeletalMeshSkeleton(
    Mesh:NiagaraDataInterfaceSkeletalMesh=$User.SkeletalMesh
    Specific Bone Index:NiagaraInt32=&MakeIntFromFloat(Float=&FN_PickFloatFromSpawnGroupCutoff(Over Cutoff Float:NiagaraFloat=1, Under Cutoff Float:NiagaraFloat=0, SpawnGroup:NiagaraInt32=$Particles.SpawnGroup, Cutoff:NiagaraInt32=1))
   )
   &SkeletalMeshSkeletonLocation(Module.Inherit Velocity=false, Module.Use Sampled Bone as Ribbon ID=true)
  ]
  particleUpdateScript [
   &ParticleState()
   &SolveForcesAndVelocity(Clamp Velocity=false, Limit Acceleration=false)
   &GenerateDeathEvent()
  ]
  rendererClasses [&NiagaraSpriteRendererProperties]
  renderers [{rendererClass=&NiagaraSpriteRendererProperties}]
 }
 FN_Contrails_Light {
  emitterSpawnScript [&FN_PreserveSkydiveUserVariables() {module=PreserveUserVariables}]
  emitterUpdateScript [
   &EmitterState(Inactive Response=::NewEnumerator1, Life Cycle Mode=::NewEnumerator1, Loop Behavior=::NewEnumerator1, Loop Duration=1, UseLoopDelay=false)
   &SpawnRateFromMultipleSpawnGroups_2() {enabled=false}
   &SpawnBurst_Instantaneous(Spawn Count=2, Use Spawn Probability=false)
  ]
  particleSpawnScript [
   &FN_SetSpawnGroup(Emitter Bone Count=2)
   &InitializeParticle(
    Color=<rgba 100 100 100 1>
    Lifetime=9999
    Module.Write Color=false
    Module.Write Lifetime=true
    Module.Write Mass=true
    Module.Write Material Random=true
    Module.Write Position=false
    Module.Write Scale=false
    Module.Write SpriteRotation=false
    Module.Write SpriteSize=true
    Position=&SimulationPosition()
    Sprite Size=&Vector2DFromFloat(Value=&UniformRangedFloat(Minimum=30, Maximum=70))
   )
   /SetVariables(Particles.Color=<rgba 10 2.028 2.9253 1>, Particles.LightExponent=1) {isSetParameters=true}
  ]
  particleUpdateScript [
   &ParticleState()
   &SampleSkeletalMeshSkeleton(Mesh:NiagaraDataInterfaceSkeletalMesh=$User.SkeletalMesh, Specific Bone Index:NiagaraInt32=$Particles.SpawnGroup)
   &SkeletalMeshSkeletonLocation(Module.Inherit Velocity=false, Module.Use Sampled Bone as Ribbon ID=true, Sampled Position Offset:Vector3f=<0 0 10>)
   &SolveForcesAndVelocity(Clamp Velocity=false, Limit Acceleration=false) {enabled=false}
   &ScaleColor(ScaleA=true, ScaleRGB=true)
  ]
  rendererClasses [&NiagaraLightRendererProperties]
  renderers [{rendererClass=&NiagaraLightRendererProperties, rendererIndex=0}]
 }
 Fireworks_Mesh {
  emitterUpdateScript [
   &EmitterScalability(
    Module.bApplySpawnCountScale=false
    Module.bApplySpawnCountScaleByDistanceFraction=false
    Module.bUseMaxDistance=false
    Module.bUseMinDistance=false
   ) {enabled=false}
   &EmitterLifeCycle(NextLoopDuration=1) {enabled=false}
   &SpawnRate(SpawnRate=20, Use Spawn Probability=false) {enabled=false}
   &EmitterState()
   &SpawnBurst_Instantaneous(Spawn Count=3, Spawn Time=0.05, Use Spawn Probability=false) {enabled=false}
  ]
  particleSpawnScript [
   &InitializeParticle(
    Color=&UniformRangedLinearColor(Minimum=<rgba 1 1 1 1>, Maximum=<rgba 1 0.5 0 1>)
    Lifetime=&UniformRangedFloat(Minimum=0.6, Maximum=0.7)
    Mesh Scale=&VectorFromFloat(Value=&Multiply_Float(A=&UniformRangedFloat(Minimum=0.225, Maximum=0.18), B=0.7))
    Module.Write Color=true
    Module.Write Lifetime=true
    Module.Write Mass=true
    Module.Write Material Random=true
    Module.Write Position=true
    Module.Write Scale=true
    Module.Write SpriteRotation=false
    Module.Write SpriteSize=false
    Position=&SimulationPosition()
    Sprite Size=<25 25>
   )
   /Niagara/Modules/Spawn/Orientation/V2/InitialMeshOrientation(
    Orientation Vector=&UniformRangedVector(Minimum=<0 0 0>, Maximum=<1 1 1>)
    Use Orientation Vector=true
    Use Rotation Vector=false
   )
  ]
  particleUpdateScript [
   &ParticleState()
   &ScaleColor(
    Scale Alpha=&FloatFromCurve(FloatCurve=DataInterface() {propertyValues {Curve {keys [{arriveTangent=-3.2421, interpMode=RCIM_Cubic, leaveTangent=-3.2421, tangentMode=RCTM_User, time=0.1015, value=1.0322}, <at 1 0>]}}})
    Scale RGB=&VectorFromCurve(VectorCurve=DataInterface() {propertyValues {
       XCurve {keys [<at -0.0011 0.01>,<at 0.2555 0>,<at 0.4988 0>,<at 1 0.5>]}
       YCurve {keys [<at -0.0011 15>,<at 0.2555 15>,<at 0.4988 5>,<at 1 1.649>]}
       ZCurve {keys [<at -0.0011 1.0067>,<at 0.2555 1>,<at 0.4988 1>,<at 0.9989 1>]}
      }})
    ScaleA=true
    ScaleRGB=true
   )
   &ScaleMeshSize(Scale Factor=&VectorFromFloat(Value=&FloatFromCurve(FloatCurve=DataInterface() {propertyValues {Curve {keys [
          {arriveTangent=9.455, interpMode=RCIM_Cubic, leaveTangent=9.455, tangentMode=RCTM_User, time=0, value=0}
          {arriveTangent=1.6472, interpMode=RCIM_Cubic, leaveTangent=1.6472, tangentMode=RCTM_User, time=0.131, value=0.6285}
          <at 1.0062 1>
         ]}}})))
   &DynamicMaterialParameters(
    Index 0 Param 1=1
    Index 0 Param 2=1
    Index 0 Param 3=&FloatFromCurve(FloatCurve=DataInterface() {propertyValues {Curve {keys [{arriveTangent=0.195, interpMode=RCIM_Cubic, leaveTangent=0.195, tangentMode=RCTM_User, time=0, value=0.03}, <at 1.0015 0.1>]}}})
    Index 0 Param 4=&FloatFromCurve(FloatCurve=DataInterface() {propertyValues {Curve {keys [<at 0 0.1>,<at 0.6927 0>]}}})
    Module.Param0WriteEnabled=true
    Module.Param1WriteEnabled=true
    Module.Param2WriteEnabled=true
    Module.Param3WriteEnabled=true
   )
  ]
  rendererClasses [&NiagaraMeshRendererProperties]
  renderers [{rendererClass=&NiagaraMeshRendererProperties, rendererIndex=0}]
 }
 L_Explosion_Hearts {
  emitterUpdateScript [
   &EmitterScalability(
    Module.bApplySpawnCountScale=false
    Module.bApplySpawnCountScaleByDistanceFraction=false
    Module.bUseMaxDistance=false
    Module.bUseMinDistance=false
   ) {enabled=false}
   &EmitterLifeCycle(NextLoopDuration=1) {enabled=false}
   &EmitterState()
   &SpawnRate(SpawnRate=20, Use Spawn Probability=false) {enabled=false}
   &SpawnBurst_Instantaneous(Spawn Count=3, Use Spawn Probability=false) {enabled=false}
  ]
  particleSpawnScript [
   &InitializeParticle(
    Color=<rgba 1 0 0.0615 1>
    Lifetime=0.8
    Module.Write Color=true
    Module.Write Lifetime=true
    Module.Write Mass=true
    Module.Write Material Random=true
    Module.Write Position=true
    Module.Write Scale=false
    Module.Write SpriteRotation=true
    Module.Write SpriteSize=true
    Position=&SimulationPosition()
    Sprite Rotation=&UniformRangedFloat(Minimum=-40, Maximum=-80)
    Sprite Size=&Multiply_Vector2(
     A=&MakeVector2D(X=&UniformAOrBFloat(A=-25, B=25), Y=50)
     B=&Vector2DFromFloat(Value=&UniformRangedFloat(Minimum=0.2, Maximum=0.3))
    )
   )
   &AddVelocityInCone(
    Cone Angle=&UniformRangedFloat(Minimum=25, Maximum=60)
    Module.Use Velocity Falloff On Cone Axis=false
    Velocity Strength=&UniformRangedFloat(Minimum=750, Maximum=800)
   ) {enabled=false}
   /Niagara/Modules/Spawn/Location/SphereLocation(MaskSpawnGroup=false, Module.Override Randomness=false, Sphere Radius=1000, Surface Only=false)
   &AddVelocityFromPoint(
    Module.Use Falloff Distance=false
    Velocity Origin:Vector3f=&SimulationPosition()
    Velocity Strength=&UniformRangedFloat(Minimum=400, Maximum=500)
   )
  ]
  particleUpdateScript [
   &CameraOffset(Camera Offset Amount=250) {enabled=false}
   &ParticleState()
   &ScaleColor(
    Scale Alpha=&FloatFromCurve(FloatCurve=DataInterface() {propertyValues {Curve {keys [
         <at 0 0>
         <at 0.0805 1.8557>
         {arriveTangent=-2.0136, interpMode=RCIM_Cubic, leaveTangent=-12.6391, tangentMode=RCTM_Break, time=0.608, value=1.86}
         <at 1.0011 0.002>
        ]}}})
    Scale RGB=&MakeVectorFromLinearColorRGB(Color=<rgba 1 1 1 1>)
    ScaleA=true
    ScaleRGB=true
   )
   &ScaleSpriteSize(Scale Factor=&Vector2DFromCurve(
     Vector2Curve=DataInterface() {propertyValues {
       XCurve {keys [
         {arriveTangent=0.1081, interpMode=RCIM_Cubic, leaveTangent=0.1081, tangentMode=RCTM_User, time=0.0715, value=0.1253}
         {arriveTangent=-3.9427, interpMode=RCIM_Cubic, leaveTangent=-3.9427, tangentMode=RCTM_User, time=0.3844, value=2.2065}
         {arriveTangent=-2.0108, interpMode=RCIM_Cubic, leaveTangent=-2.0108, time=0.7353, value=1.3595}
         {arriveTangent=0.9769, interpMode=RCIM_Cubic, leaveTangent=0.9769, tangentMode=RCTM_User, time=1.0065, value=0.9558}
        ]}
       YCurve {keys [
         <at 0 0 RCIM_Cubic>
         {arriveTangent=1.4215, interpMode=RCIM_Cubic, leaveTangent=1.4215, time=0.4034, value=1.2506}
         {arriveTangent=-1.1637, interpMode=RCIM_Cubic, leaveTangent=-1.1637, time=0.7518, value=1.0687}
         {arriveTangent=0.3291, interpMode=RCIM_Cubic, leaveTangent=0.3291, tangentMode=RCTM_User, time=1.0115, value=0.5429}
        ]}
      }}
     ScaleCurve=<1 1>
    ))
   &SpriteRotationRate(Rotation Rate=&FloatFromCurve(
     FloatCurve=DataInterface() {propertyValues {Curve {keys [
         {arriveTangent=-2.8071, interpMode=RCIM_Cubic, leaveTangent=-2.8071, tangentMode=RCTM_User, time=-0.0046, value=1.0113}
         {arriveTangent=-0.6596, interpMode=RCIM_Cubic, leaveTangent=-0.6596, tangentMode=RCTM_User, time=0.3514, value=0.3382}
         <at 1.0011 0.002>
        ]}}}
     Scale Curve=350
    ))
   &GravityForce(Gravity=<0 0 -1150>) {enabled=false}
   &Drag(Drag=5, Use Linear Drag=true, Use Rotational Drag=false)
   &SolveForcesAndVelocity(Clamp Velocity=false, Limit Acceleration=false)
   &GenerateLocationEvent(Event Send Rate=2, Module.Use Event Send Rate=true)
  ]
  rendererClasses [&NiagaraSpriteRendererProperties]
  renderers [{rendererClass=&NiagaraSpriteRendererProperties}]
 }
 L_Explosion_Mesh_Radial {
  emitterUpdateScript [
   &EmitterScalability(
    Module.bApplySpawnCountScale=false
    Module.bApplySpawnCountScaleByDistanceFraction=false
    Module.bUseMaxDistance=false
    Module.bUseMinDistance=false
   ) {enabled=false}
   &EmitterLifeCycle(MaxLoopCount=1, NextLoopDelay=0, NextLoopDuration=1.05) {enabled=false}
   &EmitterState()
   &SpawnRate(SpawnRate=50, Use Spawn Probability=false) {enabled=false}
   &SpawnBurst_Instantaneous(Spawn Count=8, Spawn Group=1, Spawn Time=0.0458, Use Spawn Probability=false) {enabled=false}
  ]
  particleSpawnScript [
   &InitializeParticle(
    Lifetime=&UniformRangedFloat(Minimum=0.5, Maximum=0.6)
    Mesh Scale=&VectorFromFloat(Value=&UniformRangedFloat(Minimum=0.165, Maximum=0.075))
    Module.Write Color=true
    Module.Write Lifetime=true
    Module.Write Mass=true
    Module.Write Material Random=true
    Module.Write Position=true
    Module.Write Scale=true
    Module.Write SpriteRotation=false
    Module.Write SpriteSize=false
    Position=&SimulationPosition()
    Sprite Size=<25 25>
   )
   /SetVariables(Particles.SVector:Vector3f=&UniformRangedVector(Minimum=<0 0 0>, Maximum=<0 1 1>)) {enabled=false, isSetParameters=true}
   &TorusLocation(
    Align Torus With Axis=true
    Large Radius=1.5
    MaskSpawnGroup=false
    Radius Position=1.2
    Torus Axis=<1 0 0>
    Torus Distribution Mode:ENiagaraTorusDistributionMode=::NewEnumerator1
    Torus Mode=::NewEnumerator2
    U Position:NiagaraFloat=&ReturnNormalizedExecIndex()
   ) {enabled=false}
   &InitialMeshOrientation(
    Rotation=&UniformRangedVector(Minimum=<0 0 0>, Maximum=<1 1 1>)
    Use Orientation Vector=true
    Use Rotation Vector=true
   )
   &DynamicMaterialParameters(
    Index 0 Param 1=&UniformRangedFloat(Minimum=1.25, Maximum=1)
    Module.Param0WriteEnabled=true
    Module.Param1WriteEnabled=false
    Module.Param2WriteEnabled=false
    Module.Param3WriteEnabled=false
   ) {enabled=false}
   &AddVelocity(Scale Added Velocity=<750 1 1>, Velocity=&Multiply_VectorByFloat(Vector=$Particles.SVector, Float=100)) {enabled=false}
   &AddVelocityFromPoint(Module.Use Falloff Distance=false, Velocity Origin:Vector3f=&SimulationPosition(), Velocity Strength=10) {enabled=false}
  ]
  particleUpdateScript [
   &Drag(
    Drag=&FloatFromCurve(
     FloatCurve=DataInterface() {propertyValues {Curve {keys [{arriveTangent=7.1572, interpMode=RCIM_Cubic, leaveTangent=7.1573, tangentMode=RCTM_User, time=-0.0053, value=0.1532}, <at 0.3232 1>]}}}
     Scale Curve=6
    )
    Use Linear Drag=false
    Use Rotational Drag=false
   )
   &SolveForcesAndVelocity(Clamp Velocity=false, Limit Acceleration=false)
   &ParticleState()
   &ScaleColor(
    Scale Alpha=&FloatFromCurve(FloatCurve=DataInterface() {propertyValues {Curve {keys [
         <at -0.0046 0.8475 RCIM_Cubic>
         {arriveTangent=-1.3398, interpMode=RCIM_Cubic, leaveTangent=-1.3398, time=0.4249, value=0.5912}
         <at 1.0011 -0.5 RCIM_Cubic>
        ]}}})
    Scale RGB=&VectorFromCurve(VectorCurve=DataInterface() {propertyValues {
       XCurve {keys [<at 0 1.5>,<at 0.2262 0.2118>,<at 0.6199 0.02>,<at 0.9212 0.02>,<at 1 0.02>]}
       YCurve {keys [<at 0 0.0009>,<at 0.1955 0.0048>,<at 0.4127 0.0092>,<at 0.6065 0.0117>,<at 0.6199 0.0118>,<at 0.9212 0.0175>,<at 1.02 0.0193>]}
       ZCurve {keys [<at 0 0.0022>,<at 0.4129 0.0102>,<at 0.6199 0.0129>,<at 1.0206 0.0196>]}
      }})
    ScaleA=true
    ScaleRGB=true
   )
   &ScaleMeshSize(Scale Factor=&VectorFromFloat(Value=&FloatFromCurve(FloatCurve=DataInterface() {propertyValues {Curve {keys [<at 0.001 0.5>,<at 0.2948 1.1652>,<at 1.0009 0.8072>]}}})))
   &DynamicMaterialParameters(
    Index 0 Param 1=1
    Index 0 Param 2=&FloatFromCurve(FloatCurve=DataInterface() {propertyValues {Curve {keys [
         {arriveTangent=-0.7948, interpMode=RCIM_Cubic, leaveTangent=-0.7949, tangentMode=RCTM_User, time=0.2519, value=0.1057}
         <at 0.7709 -10.0646 RCIM_Cubic>
        ]}}})
    Index 0 Param 4=&FloatFromCurve(FloatCurve=DataInterface() {propertyValues {Curve {keys [<at 0 6.7593 RCIM_Cubic>,<at 0.7 0 RCIM_Cubic>]}}})
    Module.Param0WriteEnabled=true
    Module.Param1WriteEnabled=true
    Module.Param2WriteEnabled=true
    Module.Param3WriteEnabled=true
   ) {module=DynamicMaterialParameters001}
  ]
  rendererClasses [&NiagaraMeshRendererProperties]
  renderers [{rendererClass=&NiagaraMeshRendererProperties, rendererIndex=0}]
 }
 L_FN_Contrails_BigHeart {
  emitterSpawnScript [&FN_PreserveSkydiveUserVariables() {module=PreserveUserVariables}]
  emitterUpdateScript [
   &EmitterState()
   &SpawnRate(Use Spawn Probability=false) {enabled=false}
   &SpawnBurst_Instantaneous(Spawn Count=1, Use Spawn Probability=false) {enabled=false}
  ]
  particleSpawnScript [
   /SetVariables(Particles.ModifyID:NiagaraInt32=&MakeIntFromFloat(Float=&Multiply_Float(
      A=&Frac_Float(Float=&Divide_Float(A=/Niagara/DynamicInputs/Execution/ReturnParticleID_AsFloat(), B=2))
      B=2
     ))) {enabled=false, isSetParameters=true}
   &FN_SetSpawnGroup(Override SpawnGroup=true)
   &InitializeParticle(
    Color=<rgba 1 0.8417 0.7639 1>
    Lifetime=&UniformRangedFloat(Minimum=0.5, Maximum=0.5)
    Mesh Scale=&Multiply_Vector(
     A=<1 1 1>
     B=&VectorFromFloat(Value=&UniformRangedFloat(Minimum=0.1, Maximum=0.25))
    )
    Module.Write Color=true
    Module.Write Lifetime=true
    Module.Write Mass=true
    Module.Write Material Random=true
    Module.Write Position=false
    Module.Write Scale=false
    Module.Write SpriteRotation=false
    Module.Write SpriteSize=false
    Position=&SimulationPosition()
    Sprite Size=<-10 10>
   )
   &CylinderLocation(
    Cylinder Height=1
    Lathe Profile=DataInterface() {propertyValues {Curve {keys []}}}
    MaskSpawnGroup=false
    Module.Enable Lathe Profile=false
    Module.Override Randomness=false
    Module.Surface Only=false
   ) {enabled=false}
   &SampleSkeletalMeshSkeleton(Mesh:NiagaraDataInterfaceSkeletalMesh=$User.SkeletalMesh, Specific Bone Index:NiagaraInt32=$Particles.SpawnGroup) {enabled=false}
   &SkeletalMeshSkeletonLocation(Module.Inherit Velocity=false, Module.Use Sampled Bone as Ribbon ID=true) {enabled=false}
   &InitialMeshOrientation(
    Orientation Vector=<0 1 0>
    Rotation=&UniformRangedVector(Minimum=<0 0.25 0>, Maximum=<0 0.25 0>)
    Use Orientation Vector=true
    Use Rotation Vector=true
   )
   /SetVariables(
    Particles.Color=<rgba 3 0 0.2346 1>
    Particles.SVector:Vector3f=&Multiply_Vector(
     A=&Subtract_Vector(A=Linked() {linkedVariable {name=Particles.Position, type=NiagaraPosition}}, B=/Niagara/DynamicInputs/Helpers/SimulationPosition())
     B=<1 0 0>
    )
    Particles.SpriteSize=&Multiply_Vector2(
     A=&MakeVector2D(X=&UniformAOrBFloat(A=-25, B=25), Y=60)
     B=&Vector2DFromFloat(Value=1)
    )
   ) {isSetParameters=true}
   &AddVelocity(Velocity=&Add_Vector(
     A=&Multiply_Vector(A=$Particles.SVector, B=&VectorFromFloat(Value=2))
     B=<0 0 200>
    ))
  ]
  particleUpdateScript [
   &CameraOffset(Camera Offset Amount=20)
   &ParticleState()
   /Niagara/Modules/Update/Orientation/V2/UpdateMeshOrientation(Rotation Rate:NiagaraFloat=$Particles.RotationOrientation, Rotation Vector:Vector3f=$Particles.RotationSpeed)
   &Drag(Drag=0.3, Use Linear Drag=true, Use Rotational Drag=false)
   /Niagara/Modules/Update/Forces/CurlNoiseForce(Noise Frequency=0.1, Noise Strength=600, Pan Noise=false, Pan Noise Field=<1 1 0>) {enabled=false}
   &DynamicMaterialParameters(
    Index 0 Param 1=&FloatFromCurve(FloatCurve=DataInterface() {propertyValues {Curve {keys [<at -0.0046 1.7>,<at 1.0011 0.002>]}}})
    Module.Param0WriteEnabled=true
    Module.Param1WriteEnabled=true
    Module.Param2WriteEnabled=true
    Module.Param3WriteEnabled=true
   )
   &SolveForcesAndVelocity(Clamp Velocity=false, Limit Acceleration=false)
   &GenerateDeathEvent()
   &ScaleColor(
    Scale RGB=&VectorFromFloat(Value=&FloatFromCurve(
      FloatCurve=DataInterface() {propertyValues {Curve {keys [<at 0 0>,<at 0.0449 0.0998>,<at 0.0949 0.01>,<at 0.1479 0.0808>,<at 0.1998 0.01>,<at 0.2417 0.0999>,<at 0.29 0.01>,<at 0.3355 0.0904>,<at 0.3768 0.01>,<at 0.4169 0.1094>,<at 0.4624 0.01>,<at 0.5036 0.1571>,<at 0.5372 0.01>,<at 0.5665 0.2716>,<at 0.599 0.01>,<at 0.6262 0.3383>,<at 0.6576 0.01>,<at 0.6836 0.4814>,<at 0.7162 0.01>,<at 0.7422 0.653>,<at 0.778 0.01>,<at 0.7975 1.5399>,<at 0.83 0.01>,<at 0.8452 2.0835>,<at 0.8745 0.01>,<at 0.8853 2.5227>,<at 0.9092 0.01>,<at 0.9254 2.8155>,<at 0.9406 0.01>,<at 0.9601 3.0984>,<at 0.9753 0.01>,<at 1 3.3717>]}}}
      Scale Curve=15
     ))
    ScaleA=true
    ScaleRGB=true
   )
   &ScaleSpriteSize(Scale Factor=&Vector2DFromFloat(Value=&FloatFromCurve(FloatCurve=DataInterface() {propertyValues {Curve {keys [<at 0 0.2 RCIM_Cubic>,<at 0.4527 0.997>,<at 1 1 RCIM_Cubic>]}}})))
   &ScaleSpriteSize(Scale Factor=&Multiply_Vector2(
     A=<1 1>
     B=&Vector2DFromFloat(Value=&FloatFromCurve(
       FloatCurve=DataInterface() {propertyValues {Curve {keys [
           <at 0.3628 1.0017>
           {arriveTangent=0.0238, time=0.4032, value=0.8167}
           {arriveTangent=0.01, leaveTangent=0.01, time=0.4372, value=1.2829}
           {arriveTangent=-0.0149, leaveTangent=-0.0149, time=0.4941, value=0.7085}
           {arriveTangent=0.7936, leaveTangent=0.7936, time=0.5178, value=1}
           <at 0.6288 1>
           <at 0.6664 0.6764>
           <at 0.7156 1.8173>
           <at 0.7792 0.3608>
           <at 0.8177 1>
           <at 1 1 RCIM_Cubic>
          ]}}}
       Scale Curve=1.2
      ))
    )) {module=ScaleSpriteSize001}
  ]
  rendererClasses [&NiagaraSpriteRendererProperties]
  renderers [{rendererClass=&NiagaraSpriteRendererProperties}]
 }
 L_FN_Mesh_Boom {
  emitterUpdateScript [
   &EmitterScalability(
    Module.bApplySpawnCountScale=false
    Module.bApplySpawnCountScaleByDistanceFraction=false
    Module.bUseMaxDistance=false
    Module.bUseMinDistance=false
   ) {enabled=false}
   &EmitterLifeCycle(NextLoopDuration=1) {enabled=false}
   &EmitterState()
   &SpawnRate(SpawnRate=0, Use Spawn Probability=false) {enabled=false}
   &SpawnBurst_Instantaneous(Spawn Count=1, Use Spawn Probability=false) {enabled=false}
  ]
  particleSpawnScript [
   &InitializeParticle(
    Lifetime=0.8
    Mesh Scale=&VectorFromFloat(Value=0.13)
    Module.Write Color=true
    Module.Write Lifetime=true
    Module.Write Mass=true
    Module.Write Material Random=true
    Module.Write Position=true
    Module.Write Scale=true
    Module.Write SpriteRotation=false
    Module.Write SpriteSize=false
    Position=/Niagara/Functions/Localspace/SimulationPosition()
    Sprite Size=<25 25>
   )
   &InitialMeshOrientation(
    Orientation Axis:Vector3f=<0 0 1>
    Orientation Vector=<0 0 1>
    Rotation=&MakeVector(X=$System.RandomVal, Y=0, Z=-0.25)
    Rotation Coordinate Space:ENiagaraMeshTransforms=::NewEnumerator0
    Use Orientation Vector=true
    Use Rotation Vector=true
   ) {enabled=false}
  ]
  particleUpdateScript [
   &ParticleState()
   &ScaleMeshSize(Scale Factor=/Niagara/DynamicInputs/ScaleByCurve/ScaleVectorByCurve(
     InitialValue:Vector3f=<1 1 1>
     ScaleCurve:NiagaraDataInterfaceVectorCurve=DataInterface() {propertyValues {
       XCurve {keys [
         <at 0 0 RCIM_Cubic>
         {arriveTangent=2.2843, interpMode=RCIM_Cubic, leaveTangent=2.2843, time=0.3178, value=1.2}
         {arriveTangent=-0.4627, interpMode=RCIM_Cubic, leaveTangent=-0.4627, time=0.4378, value=1}
         {arriveTangent=-1.7786, interpMode=RCIM_Cubic, leaveTangent=-1.7786, time=0.75, value=1}
         <at 1 0 RCIM_Cubic>
        ]}
       YCurve {keys [
         <at 0 0 RCIM_Cubic>
         {arriveTangent=2.2886, interpMode=RCIM_Cubic, leaveTangent=2.2886, time=0.3167, value=1.4998}
         {arriveTangent=-1.1535, interpMode=RCIM_Cubic, leaveTangent=-1.1535, time=0.5159, value=1.1808}
         {arriveTangent=-2.4394, interpMode=RCIM_Cubic, leaveTangent=-2.4394, time=0.75, value=1}
         <at 1 0 RCIM_Cubic>
        ]}
       ZCurve {keys [
         <at 0 0 RCIM_Cubic>
         {arriveTangent=3.0427, interpMode=RCIM_Cubic, leaveTangent=3.0427, time=0.25, value=1.3258}
         {arriveTangent=-0.4, interpMode=RCIM_Cubic, leaveTangent=-0.4, time=0.37, value=1.1258}
         {arriveTangent=-1.787, interpMode=RCIM_Cubic, leaveTangent=-1.787, time=0.75, value=1.1258}
         <at 1 0 RCIM_Cubic>
        ]}
      }}
     CurveIndex:NiagaraFloat=$Particles.NormalizedAge
    ))
   &ScaleColor(Scale RGB=&MakeVectorFromLinearColorRGB(Color=<rgba 1 1 1 1>), ScaleA=true, ScaleRGB=true)
   &DynamicMaterialParameters(
    Index 0 Param 1=&FloatFromCurve(FloatCurve=DataInterface() {propertyValues {Curve {keys [
         {arriveTangent=3.8817, interpMode=RCIM_Cubic, leaveTangent=3.8817, tangentMode=RCTM_User, time=0, value=0}
         {arriveTangent=0.7221, interpMode=RCIM_Cubic, leaveTangent=0.7221, tangentMode=RCTM_User, time=0.7564, value=2.091}
         {arriveTangent=4.2941, interpMode=RCIM_Cubic, leaveTangent=4.2941, tangentMode=RCTM_User, time=0.9228, value=2.078}
        ]}}})
    Index 0 Param 2=1
    Index 0 Param 3=&FloatFromCurve(FloatCurve=DataInterface() {propertyValues {Curve {keys [
         {arriveTangent=-17.7515, interpMode=RCIM_Cubic, leaveTangent=-17.7515, tangentMode=RCTM_User, time=-0.0046, value=1.0113}
         {arriveTangent=-0.2845, interpMode=RCIM_Cubic, leaveTangent=-0.2845, tangentMode=RCTM_User, time=0.1, value=0.065}
         {arriveTangent=0.4619, interpMode=RCIM_Cubic, leaveTangent=0.4619, tangentMode=RCTM_User, time=0.6513, value=0.0607}
         {arriveTangent=3.5589, interpMode=RCIM_Cubic, leaveTangent=3.5589, tangentMode=RCTM_User, time=1, value=1}
        ]}}})
    Index 0 Param 4=100
    Module.Param0WriteEnabled=true
    Module.Param1WriteEnabled=true
    Module.Param2WriteEnabled=true
    Module.Param3WriteEnabled=true
   )
   &GravityForce(Gravity=&Lerp_Vector(
     A=<0 0 0>
     B=<0 0 -980>
     Alpha=&FloatFromCurve(FloatCurve=DataInterface() {propertyValues {Curve {keys [<at 0.65 0>,<at 0.675 1>]}}})
    ))
   &SolveForcesAndVelocity(Clamp Velocity=false, Limit Acceleration=false)
  ]
  rendererClasses [&NiagaraSpriteRendererProperties, &NiagaraMeshRendererProperties]
  renderers [
   {rendererClass=&NiagaraSpriteRendererProperties}
   {rendererClass=&NiagaraMeshRendererProperties, rendererIndex=1}
  ]
 }
 L_FN_Mesh_Ka {
  emitterUpdateScript [
   &EmitterScalability(
    Module.bApplySpawnCountScale=false
    Module.bApplySpawnCountScaleByDistanceFraction=false
    Module.bUseMaxDistance=false
    Module.bUseMinDistance=false
   ) {enabled=false}
   &EmitterLifeCycle(NextLoopDuration=1) {enabled=false}
   &EmitterState()
   &SpawnRate(SpawnRate=0, Use Spawn Probability=false) {enabled=false}
   &SpawnBurst_Instantaneous(Spawn Count=1, Use Spawn Probability=false) {enabled=false}
  ]
  particleSpawnScript [
   &InitializeParticle(
    Lifetime=0.75
    Mesh Scale=&VectorFromFloat(Value=0.13)
    Module.Write Color=true
    Module.Write Lifetime=true
    Module.Write Mass=true
    Module.Write Material Random=true
    Module.Write Position=true
    Module.Write Scale=true
    Module.Write SpriteRotation=true
    Module.Write SpriteSize=false
    Position=&SimulationPosition()
    Sprite Size=<25 25>
   )
   &InitialMeshOrientation(
    Orientation Axis:Vector3f=<0 0 1>
    Orientation Vector=<0 0 1>
    Rotation=&MakeVector(X=$System.RandomVal, Y=0, Z=0)
    Rotation Coordinate Space:ENiagaraMeshTransforms=::NewEnumerator0
    Use Orientation Vector=true
    Use Rotation Vector=true
   ) {enabled=false}
  ]
  particleUpdateScript [
   &ParticleState()
   &ScaleMeshSize(Scale Factor=&ScaleVectorByCurve(
     InitialValue:Vector3f=<1 1 1>
     ScaleCurve:NiagaraDataInterfaceVectorCurve=DataInterface() {propertyValues {
       XCurve {keys [
         <at 0 0 RCIM_Cubic>
         {arriveTangent=2.2843, interpMode=RCIM_Cubic, leaveTangent=2.2843, time=0.3178, value=1.2}
         {arriveTangent=-0.4851, interpMode=RCIM_Cubic, leaveTangent=-0.4851, time=0.4378, value=1}
         {arriveTangent=-1.7786, interpMode=RCIM_Cubic, leaveTangent=-1.7786, time=0.75, value=0.9903}
         <at 1 0 RCIM_Cubic>
        ]}
       YCurve {keys [
         <at 0 0 RCIM_Cubic>
         {arriveTangent=3.2407, interpMode=RCIM_Cubic, leaveTangent=3.2407, time=0.25, value=1.3991}
         {arriveTangent=-0.6628, interpMode=RCIM_Cubic, leaveTangent=-0.6628, time=0.37, value=1.1991}
         {arriveTangent=-1.9033, interpMode=RCIM_Cubic, leaveTangent=-1.9033, time=0.75, value=1.0676}
         <at 1 0 RCIM_Cubic>
        ]}
       ZCurve {keys [
         <at 0 0 RCIM_Cubic>
         {arriveTangent=3.2407, interpMode=RCIM_Cubic, leaveTangent=3.2407, time=0.25, value=1.3991}
         {arriveTangent=-0.6628, interpMode=RCIM_Cubic, leaveTangent=-0.6628, time=0.37, value=1.1991}
         {arriveTangent=-1.9033, interpMode=RCIM_Cubic, leaveTangent=-1.9033, time=0.75, value=1.0676}
         <at 1 0 RCIM_Cubic>
        ]}
      }}
     CurveIndex:NiagaraFloat=$Particles.NormalizedAge
    ))
   &ScaleColor(Scale RGB=&MakeVectorFromLinearColorRGB(Color=<rgba 1 1 1 1>), ScaleA=true, ScaleRGB=true)
   &DynamicMaterialParameters(
    Index 0 Param 1=&FloatFromCurve(FloatCurve=DataInterface() {propertyValues {Curve {keys [
         {arriveTangent=5.6948, interpMode=RCIM_Cubic, leaveTangent=5.6948, tangentMode=RCTM_User, time=0, value=0}
         {arriveTangent=-0.5168, interpMode=RCIM_Cubic, leaveTangent=-0.5168, tangentMode=RCTM_User, time=0.8442, value=1.658}
         {interpMode=RCIM_Cubic, leaveTangent=0.574, time=0.98, value=1.3}
        ]}}})
    Index 0 Param 3=&FloatFromCurve(FloatCurve=DataInterface() {propertyValues {Curve {keys [
         {arriveTangent=-6.3759, interpMode=RCIM_Cubic, leaveTangent=-6.3759, tangentMode=RCTM_User, time=-0.0046, value=1.0113}
         {arriveTangent=-0.0113, interpMode=RCIM_Cubic, leaveTangent=-0.0113, time=0.4003, value=0.002}
         {arriveTangent=0.0913, interpMode=RCIM_Cubic, leaveTangent=0.0913, tangentMode=RCTM_User, time=1, value=1}
        ]}}})
    Index 0 Param 4=80
    Module.Param0WriteEnabled=true
    Module.Param1WriteEnabled=true
    Module.Param2WriteEnabled=true
    Module.Param3WriteEnabled=true
   )
   &GravityForce(Gravity=&Lerp_Vector(
     A=<0 0 0>
     B=<0 0 -980>
     Alpha=&FloatFromCurve(FloatCurve=DataInterface() {propertyValues {Curve {keys [<at 0.65 0>,<at 0.675 1>]}}})
    ))
   &SolveForcesAndVelocity(Clamp Velocity=false, Limit Acceleration=false)
   /SetVariables(Particles.Position=&Add_Vector(A=Linked() {linkedVariable {name=Particles.Position, type=NiagaraPosition}}, B=<0 0 0.7>)) {isSetParameters=true}
  ]
  rendererClasses [&NiagaraMeshRendererProperties]
  renderers [{rendererClass=&NiagaraMeshRendererProperties, rendererIndex=0}]
 }
 L_SmokeyFingers {
  emitterUpdateScript [
   &EmitterScalability(
    Module.bApplySpawnCountScale=false
    Module.bApplySpawnCountScaleByDistanceFraction=false
    Module.bUseMaxDistance=false
    Module.bUseMinDistance=false
   ) {enabled=false}
   &EmitterLifeCycle(MaxLoopCount=1, NextLoopDuration=1) {enabled=false}
   &EmitterState()
   &SpawnRate(SpawnRate=20, Use Spawn Probability=false) {enabled=false}
   &SpawnBurst_Instantaneous(Spawn Count=8, Use Spawn Probability=false) {enabled=false}
  ]
  particleSpawnScript [
   &InitializeParticle(
    Color=<rgba 0 1 1 0.85>
    Lifetime=&UniformRangedFloat(Minimum=0.45, Maximum=0.6)
    Mesh Scale=&UniformRangedVector(Minimum=<1 25 25>, Maximum=<1.5 25 25>)
    Module.Write Color=true
    Module.Write Lifetime=true
    Module.Write Mass=true
    Module.Write Material Random=true
    Module.Write Position=true
    Module.Write Scale=true
    Module.Write SpriteRotation=false
    Module.Write SpriteSize=false
    Position=&SimulationPosition()
    Sprite Size=<25 25>
   )
   &TorusLocation(
    Align Torus With Axis=true
    Large Radius=5
    MaskSpawnGroup=false
    Torus Axis=<1 0 0>
    Torus Distribution Mode:ENiagaraTorusDistributionMode=::NewEnumerator1
    Torus Mode=::NewEnumerator2
    U Position:NiagaraFloat=&Subtract_Float(A=&ReturnNormalizedExecIndex(), B=2)
   ) {enabled=false}
   &AddVelocityFromPoint(
    Module.Use Falloff Distance=false
    Origin Offset Coordinate Space=::Simulation
    Velocity Origin:Vector3f=&SimulationPosition()
    Velocity Strength=&UniformRangedFloat(Minimum=1, Maximum=5)
   )
  ]
  particleUpdateScript [
   &GravityForce(Gravity=<0 0 -0.01>)
   &SolveForcesAndVelocity(Clamp Velocity=false, Limit Acceleration=false)
   &ParticleState()
   &ScaleColor(
    Scale Alpha=&FloatFromCurve(FloatCurve=DataInterface() {propertyValues {Curve {keys [
         {arriveTangent=200.5769, interpMode=RCIM_Cubic, leaveTangent=200.5752, tangentMode=RCTM_User, time=0, value=0}
         {arriveTangent=0.0668, interpMode=RCIM_Cubic, leaveTangent=0.0668, tangentMode=RCTM_User, time=0.01, value=1}
         {arriveTangent=-0.2664, interpMode=RCIM_Cubic, leaveTangent=-0.2664, tangentMode=RCTM_User, time=0.427, value=0.9565}
         <at 1 0>
        ]}}})
    Scale RGB=&VectorFromCurve(VectorCurve=DataInterface() {propertyValues {
       bUseLUT=false
       XCurve {keys [
         <at -0.0009 0.9906 RCIM_Cubic>
         {arriveTangent=-1.9223, interpMode=RCIM_Cubic, leaveTangent=-1.0067, tangentMode=RCTM_Break, time=0.3779, value=0.4}
         <at 1 0 RCIM_Cubic>
        ]}
       YCurve {keys [
         {arriveTangent=-3.6896, interpMode=RCIM_Cubic, leaveTangent=-3.6896, tangentMode=RCTM_User, time=-0.0011, value=1.0067}
         {arriveTangent=-0.3437, interpMode=RCIM_Cubic, leaveTangent=-0.3437, tangentMode=RCTM_User, time=0.3551, value=0.2247}
         <at 1 0 RCIM_Cubic>
        ]}
       ZCurve {keys [
         <at -0.0011 0 RCIM_Cubic>
         {arriveTangent=-0.223, interpMode=RCIM_Cubic, leaveTangent=-0.223, tangentMode=RCTM_User, time=0.9965, value=0.7415}
        ]}
      }})
    ScaleA=true
    ScaleRGB=true
   )
   &ScaleMeshSize(Scale Factor=&VectorFromFloat(Value=&FloatFromCurve(FloatCurve=DataInterface() {propertyValues {Curve {keys [
          {arriveTangent=4.641, interpMode=RCIM_Cubic, leaveTangent=4.641, tangentMode=RCTM_User, time=0, value=0}
          {arriveTangent=0.8503, interpMode=RCIM_Cubic, leaveTangent=0.8503, time=0.1278, value=0.5491}
          <at 1 0.8503 RCIM_Cubic>
         ]}}})))
   &DynamicMaterialParameters(
    Index 0 Param 1=&FloatFromCurve(FloatCurve=DataInterface() {propertyValues {Curve {keys [
         {arriveTangent=3.1529, interpMode=RCIM_Cubic, leaveTangent=3.1529, tangentMode=RCTM_User, time=0, value=0.5}
         {arriveTangent=0.4946, interpMode=RCIM_Cubic, leaveTangent=0.4946, tangentMode=RCTM_User, time=0.1931, value=0.9256}
         <at 1.001 1.1573>
        ]}}})
    Index 0 Param 2=&FloatFromCurve(FloatCurve=DataInterface() {propertyValues {Curve {keys [<at 0.505 0.9841>,<at 0.7325 0.0093>,<at 1.0011 0.002>]}}})
    Index 0 Param 3=&FloatFromCurve(FloatCurve=DataInterface() {propertyValues {Curve {keys [
         {arriveTangent=3.7271, interpMode=RCIM_Cubic, leaveTangent=3.7271, tangentMode=RCTM_User, time=-0.0046, value=1.3681}
         {arriveTangent=0.9941, interpMode=RCIM_Cubic, leaveTangent=0.9941, time=0.3075, value=2.1762}
         <at 1 2.3668>
        ]}}})
    Module.Param0WriteEnabled=true
    Module.Param1WriteEnabled=true
    Module.Param2WriteEnabled=true
    Module.Param3WriteEnabled=true
   )
  ]
  rendererClasses [&NiagaraMeshRendererProperties]
  renderers [{rendererClass=&NiagaraMeshRendererProperties, rendererIndex=0}]
 }
 Swirling_hearts {
  emitterSpawnScript [&FN_PreserveSkydiveUserVariables() {module=PreserveUserVariables}]
  emitterUpdateScript [
   &EmitterState()
   &SpawnRateFromMultipleSpawnGroups_2(Global Spawnrate=5)
  ]
  particleSpawnScript [
   &FN_SetSpawnGroup(Emitter Bone Count=2)
   &InitializeParticle(
    Color=<rgba 1 0.1583 0.0084 1>
    Lifetime=&UniformRangedFloat(Minimum=0.4, Maximum=0.6)
    Mesh Scale=&Multiply_Vector(
     A=<1 1 1>
     B=&VectorFromFloat(Value=&UniformRangedFloat(Minimum=0.2, Maximum=1))
    )
    Module.Write Color=true
    Module.Write Lifetime=true
    Module.Write Mass=true
    Module.Write Material Random=true
    Module.Write Position=false
    Module.Write Scale=true
    Module.Write SpriteRotation=false
    Module.Write SpriteSize=true
    Position=&SimulationPosition()
    Sprite Size=&MakeVector2D(X=&UniformAOrBFloat(A=-10, B=10), Y=10)
   )
   &CylinderLocation(
    Cylinder Height=1
    Lathe Profile=DataInterface() {propertyValues {Curve {keys []}}}
    MaskSpawnGroup=false
    Module.Enable Lathe Profile=false
    Module.Override Randomness=false
    Module.Surface Only=false
   )
   &SampleSkeletalMeshSkeleton(Mesh:NiagaraDataInterfaceSkeletalMesh=$User.SkeletalMesh, Specific Bone Index:NiagaraInt32=$Particles.SpawnGroup)
   &SkeletalMeshSkeletonLocation(Module.Inherit Velocity=false, Module.Use Sampled Bone as Ribbon ID=true)
   &AddVelocity(Velocity=&UniformRangedVector(Minimum=<0 0 100>, Maximum=<1 1 80>))
   &InitialMeshOrientation(
    Orientation Vector=&UniformRangedVector(Minimum=<-1 -1 -1>, Maximum=<1 1 1>)
    Use Orientation Vector=true
    Use Rotation Vector=false
   )
   /SetVariables(Particles.Color=&MakeLinearColorFromVectorAndFloat(
     "Vector (RGB)"=&MakeVectorFromLinearColorRGB(Color=&ColorFromCurve(ColorCurve=DataInterface() {propertyValues {
         AlphaCurve {keys [<at -0.0023 1 RCIM_Cubic>,<at 0.3756 1>,<at 0.9977 1 RCIM_Cubic>]}
         BlueCurve {keys [<at 0.3756 0.0782>,<at 0.4415 0.0786 RCIM_Cubic>]}
         GreenCurve {keys [<at 0.3756 0>,<at 0.4415 0 RCIM_Cubic>]}
         RedCurve {keys [<at 0.3756 1>,<at 0.4415 1 RCIM_Cubic>]}
        }}))
     "Float (Alpha)"=1
    )) {isSetParameters=true}
   &InheritVelocity(Module.Limit Speed=false, Velocity Scale=<1 1 1.05>) {enabled=false}
   &SphereLocation(MaskSpawnGroup=false, Module.Override Randomness=false, Non Uniform Scale=<1.5 1.5 1.5>, Sphere Radius=20, Surface Only=false)
  ]
  particleUpdateScript [
   &ParticleState()
   &Drag(Drag=0.5, Use Linear Drag=true, Use Rotational Drag=false)
   &CurlNoiseForce(
    Noise Frequency:NiagaraFloat=&UniformRangedFloat(Minimum=0.1, Maximum=1)
    Noise Strength=&UniformRangedFloat(Minimum=300, Maximum=600)
    Pan Noise=false
    Pan Noise Field=<1 1 0>
   )
   &SolveForcesAndVelocity(Clamp Velocity=false, Limit Acceleration=false)
   &ScaleSpriteSize(Scale Factor=&Multiply_Vector2(
     A=<1 1>
     B=&Vector2DFromFloat(Value=&FloatFromCurve(FloatCurve=DataInterface() {propertyValues {Curve {keys [
           <at 0 0.15 RCIM_Cubic>
           {arriveTangent=0.0667, leaveTangent=0.0667, time=0.1259, value=1}
           {arriveTangent=-1.1418, interpMode=RCIM_Cubic, leaveTangent=-1.1418, time=0.8397, value=1}
           {interpMode=RCIM_Cubic, tangentMode=RCTM_User, time=1, value=0.002}
          ]}}}))
    ))
   &ScaleSpriteSize(Scale Factor=&Multiply_Vector2(
     A=<1 1>
     B=&Vector2DFromFloat(Value=&FloatFromCurve(
       FloatCurve=DataInterface() {propertyValues {Curve {keys [<at 0.4336 1.0007>,<at 0.4757 0.8507>,<at 0.5438 1.5807>,<at 0.6167 0.9007>,<at 0.6539 1.0007>,<at 1 1 RCIM_Cubic>]}}}
       Scale Curve=1.2
      ))
    )) {module=ScaleSpriteSize001}
  ]
  rendererClasses [&NiagaraSpriteRendererProperties]
  renderers [{rendererClass=&NiagaraSpriteRendererProperties}]
 }
}
summary {
 systemName=NS_Contrail_TNTina_V2
 emitters [
  {emitterName=L_FN_Contrails_BigHeart, rendererClasses [&NiagaraSpriteRendererProperties]}
  {emitterName=Swirling_hearts, rendererClasses [&NiagaraSpriteRendererProperties]}
  {emitterName=FN_Contrails_AttachedParticle, rendererClasses [&NiagaraSpriteRendererProperties]}
  {emitterName=FN_Contrails_Light, rendererClasses [&NiagaraLightRendererProperties]}
  {
   emitterName=Contrail_SpeedLine
   rendererClasses [&NiagaraSpriteRendererProperties, &NiagaraRibbonRendererProperties]
  }
  {emitterName=L_Explosion_Mesh_Radial, rendererClasses [&NiagaraMeshRendererProperties]}
  {emitterName=L_SmokeyFingers, rendererClasses [&NiagaraMeshRendererProperties]}
  {emitterName=L_Explosion_Hearts, rendererClasses [&NiagaraSpriteRendererProperties]}
  {emitterName=L_FN_Mesh_Ka, rendererClasses [&NiagaraMeshRendererProperties]}
  {emitterName=L_FN_Mesh_Boom, rendererClasses [&NiagaraSpriteRendererProperties, &NiagaraMeshRendererProperties]}
  {emitterName=FN_Contrails_LandR_Parents, rendererClasses [&NiagaraSpriteRendererProperties]}
  {emitterName=FN_Contrails_LandR_Parents_delayed, rendererClasses [&NiagaraSpriteRendererProperties]}
  {emitterName=Fireworks_Mesh, rendererClasses [&NiagaraMeshRendererProperties]}
 ]
 userVariables [
  User.bIsFrontEnd {type=NiagaraBool}
  User.bIsFrontEndPreview {type=NiagaraBool}
  User.FEVec {type=Vector3f, description="", defaultValue {struct=Vector3f, value=<0 0 1000>}}
  User.PlayerSpeed {type=NiagaraFloat, defaultValue=0, description=""}
  User.RibbonIDOffset {type=NiagaraInt32}
  User.SkeletalMesh {
   type=NiagaraDataInterfaceSkeletalMesh
   defaultValue {
    struct=NiagaraExt_VariableValue_DataInterface
    value {dataInterface=&self:NiagaraDataInterfaceSkeletalMesh_0}
   }
  }
  User.SpawnRate {type=NiagaraFloat, defaultValue=360, description=""}
 ]
}
